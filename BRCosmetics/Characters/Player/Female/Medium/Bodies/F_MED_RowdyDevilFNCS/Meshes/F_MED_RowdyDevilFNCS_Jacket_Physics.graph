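# Physics asset for the F_MED_RowdyDevilFNCS female character's jacket, defining bodies and constraints for secondary motion on five flaps: three back, two front.
# Capsules on pelvis, spine_01 and spine_02 form a locked torso chain, with a tiny MSTR anchor capsule per flap locked to spine_02.
# Each flap capsule hangs from its anchor with asymmetric swing limits and almost no twist, with front flaps allowed wider swing than back ones.
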PhysicsAsset /BRCosmetics/Characters/Player/Female/Medium/Bodies/F_MED_RowdyDevilFNCS/Meshes/F_MED_RowdyDevilFNCS_Jacket_Physics
bodies {
 dyn_jacket_bk [{shapeType=Capsule, length=6, radius=3, shapeName=dyn_jacket_bk_capsule, center=<5.2 1 0>, rotation=<rot 5 -90 0>}]
 dyn_jacket_bk_MSTR [{shapeType=Capsule, length=0.505, radius=0.505, shapeName=dyn_jacket_bk_MSTR_capsule, rotation=<rot 0 -90 0>}]
 dyn_jacket_bk_l [{shapeType=Capsule, length=6, radius=3, shapeName=dyn_jacket_bk_l_capsule, center=<5.2059 0 0>, rotation=<rot 4.4384 -90 6.3014>}]
 dyn_jacket_bk_l_MSTR [{shapeType=Capsule, length=0.505, radius=0.505, shapeName=dyn_jacket_bk_l_MSTR_capsule, rotation=<rot 0 -90 0>}]
 dyn_jacket_bk_r [{shapeType=Capsule, length=6, radius=3, shapeName=dyn_jacket_bk_r_capsule, center=<5.3971 0 0>, rotation=<rot 2.6081 -90 -4>}]
 dyn_jacket_bk_r_MSTR [{shapeType=Capsule, length=0.505, radius=0.505, shapeName=dyn_jacket_bk_r_MSTR_capsule, rotation=<rot 0 -90 0>}]
 dyn_jacket_fr_l [{shapeType=Capsule, length=6, radius=3, shapeName=dyn_jacket_fr_l_capsule, center=<7.2859 0 0>, rotation=<rot 10 -89.2864 7>}]
 dyn_jacket_fr_l_MSTR [{shapeType=Capsule, length=0.505, radius=0.505, shapeName=dyn_jacket_fr_l_MSTR_capsule, rotation=<rot 0 -90 0>}]
 dyn_jacket_fr_r [{shapeType=Capsule, length=6, radius=3, shapeName=dyn_jacket_fr_r_capsule, center=<6.8446 0 0>, rotation=<rot -12.11 -90 -11>}]
 dyn_jacket_fr_r_MSTR [{shapeType=Capsule, length=0.505, radius=0.505, shapeName=dyn_jacket_fr_r_MSTR_capsule, rotation=<rot 0 -90 0>}]
 pelvis [{shapeType=Capsule, length=9.6013, radius=10.1086, shapeName=pelvis_capsule, center=<1.0718 -1.5342 0.3595>}]
 spine_01 [{shapeType=Capsule, length=8, radius=7.5, shapeName=spine_01_capsule, center=<0 1.1003 0>}]
 spine_02 [{shapeType=Capsule, length=4.5, radius=7.5, shapeName=spine_02_capsule, center=<-1.149 1.7683 -0.0113>, rotation=<rot -0.6759 -90 0.3133>}]
}
constraints [
 {
  bone1Name=dyn_jacket_bk_MSTR
  bone2Name=spine_02
  swing1LimitDegrees=45
  swing1Motion=Locked
  swing2LimitDegrees=45
  swing2Motion=Locked
  twistLimitDegrees=45
  twistMotion=Locked
 }
 {
  bone1Name=dyn_jacket_bk_l_MSTR
  bone2Name=spine_02
  swing1LimitDegrees=45
  swing1Motion=Locked
  swing2LimitDegrees=45
  swing2Motion=Locked
  twistLimitDegrees=45
  twistMotion=Locked
 }
 {
  bone1Name=dyn_jacket_bk_r_MSTR
  bone2Name=spine_02
  swing1LimitDegrees=45
  swing1Motion=Locked
  swing2LimitDegrees=45
  swing2Motion=Locked
  twistLimitDegrees=45
  twistMotion=Locked
 }
 {
  bone1Name=dyn_jacket_fr_l_MSTR
  bone2Name=spine_02
  swing1LimitDegrees=45
  swing1Motion=Locked
  swing2LimitDegrees=45
  swing2Motion=Locked
  twistLimitDegrees=45
  twistMotion=Locked
 }
 {
  bone1Name=dyn_jacket_fr_r_MSTR
  bone2Name=spine_02
  swing1LimitDegrees=45
  swing1Motion=Locked
  swing2LimitDegrees=45
  swing2Motion=Locked
  twistLimitDegrees=45
  twistMotion=Locked
 }
 {
  bone1Name=spine_01
  bone2Name=pelvis
  swing1LimitDegrees=45
  swing1Motion=Locked
  swing2LimitDegrees=45
  swing2Motion=Locked
  twistLimitDegrees=45
  twistMotion=Locked
 }
 {
  bone1Name=spine_02
  bone2Name=spine_01
  swing1LimitDegrees=45
  swing1Motion=Locked
  swing2LimitDegrees=45
  swing2Motion=Locked
  twistLimitDegrees=45
  twistMotion=Locked
 }
 {bone1Name=dyn_jacket_bk, bone2Name=dyn_jacket_bk_MSTR, swing1LimitDegrees=30, swing2LimitDegrees=5, twistLimitDegrees=1}
 {bone1Name=dyn_jacket_bk_l, bone2Name=dyn_jacket_bk_l_MSTR, swing1LimitDegrees=30, swing2LimitDegrees=5, twistLimitDegrees=1}
 {bone1Name=dyn_jacket_bk_r, bone2Name=dyn_jacket_bk_r_MSTR, swing1LimitDegrees=30, swing2LimitDegrees=5, twistLimitDegrees=1}
 {bone1Name=dyn_jacket_fr_l, bone2Name=dyn_jacket_fr_l_MSTR, swing1LimitDegrees=45, swing2LimitDegrees=10, twistLimitDegrees=1}
 {bone1Name=dyn_jacket_fr_r, bone2Name=dyn_jacket_fr_r_MSTR, swing1LimitDegrees=45, swing2LimitDegrees=10, twistLimitDegrees=1}
]
