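# Physics asset for the necklace on the RowdyDevilFNCS female medium character body, set up for secondary motion of the dangling piece.
# Capsules on spine_04 and spine_05 form the torso volume, with small capsules on the necklace master and two chain bones.
# Locked angular constraints tie spine_05 to spine_04 and the master to spine_05.
# The two chain joints are left unlocked, with zero twist limits and a narrower secondary swing limit down the chain.
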PhysicsAsset /BRCosmetics/Characters/Player/Female/Medium/Bodies/F_MED_RowdyDevilFNCS/Meshes/F_MED_RowdyDevilFNCS_NecklacePhysics
bodies {
 dyn_necklace_01 [{shapeType=Capsule, length=2, radius=1.5, shapeName=dyn_necklace_01_capsule, center=<1 -0.3393 0>, rotation=<rot 0 -90 0>}]
 dyn_necklace_02 [{shapeType=Capsule, length=2.8071, radius=1.3309, shapeName=dyn_necklace_02_capsule, center=<2 0 0>, rotation=<rot 0 -91.1015 -89.9262>}]
 dyn_necklace_MSTR [{shapeType=Capsule, length=0.5, radius=0.5, shapeName=dyn_necklace_MSTR_capsule}]
 spine_04 [
  {shapeType=Capsule, length=9.0076, radius=10.3916, center=<5.6234 -0.6779 0>}
  {shapeType=Capsule, length=10.0312, radius=9.2812, center=<-1.2326 1.308 0>, rotation=<rot 90 0 0>}
 ]
 spine_05 [{shapeType=Capsule, length=9.678, radius=9.8571, center=<-0.0844 -0.9901 0>}]
}
constraints [
 {
  bone1Name=spine_05
  bone2Name=spine_04
  swing1LimitDegrees=45
  swing1Motion=Locked
  swing2LimitDegrees=45
  swing2Motion=Locked
  twistLimitDegrees=45
  twistMotion=Locked
 }
 {
  bone1Name=dyn_necklace_MSTR
  bone2Name=spine_05
  swing1LimitDegrees=45
  swing1Motion=Locked
  swing2LimitDegrees=45
  swing2Motion=Locked
  twistLimitDegrees=45
  twistMotion=Locked
 }
 {bone1Name=dyn_necklace_01, bone2Name=dyn_necklace_MSTR, swing1LimitDegrees=45, swing2LimitDegrees=15, twistLimitDegrees=0}
 {bone1Name=dyn_necklace_02, bone2Name=dyn_necklace_01, swing1LimitDegrees=45, swing2LimitDegrees=0, twistLimitDegrees=0}
]
