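# Post-copy AnimBP for the F_MED_RowdyDevilFNCS female cosmetic body, adding secondary motion to jacket, necklace and shoelaces.
# The attached parent's pose gets a right-calf constraint on the shoelace master, then RigidBody jacket and necklace simulations and a Control Rig.
# That cached pose feeds two AnimDynamics shoelace chains, selected by bIsInFrontEnd: world-space when true, root-relative otherwise.
# An unconnected shoelace Modify Bone node remains.
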AnimBlueprint /BRCosmetics/Characters/Player/Female/Medium/Bodies/F_MED_RowdyDevilFNCS/Meshes/F_MED_RowdyDevilFNCS_AnimBP
target_skeleton=&/F_MED_RowdyDevilFNCS_Skeleton
graphs [
 AnimGraph {nodes [
   "Transform (Modify) Bone - Bone: dyn_shoelace"!(ComponentPose, bAlphaBoolEnabled:bool=True, Rotation="0,0,-80") -> (Pose) {
    alphaInputType=Bool
    rotationMode=BMM_Additive
    rotationSpace=BCS_BoneSpace
    alphaBoolBlend {blendOption=Cubic}
    boneToModify {boneName=dyn_shoelace}
    rotation {roll=-80}
   }
   n1 = Anim Dynamics - Bone: dyn_shoelace_01_l(ComponentPose=Local To Component(LocalPose=Use cached pose 'PhysPose'()), Alpha=1, ExternalForce="0,0,0") {
    angularDampingOverride=0.999
    bOverrideAngularDamping=true
    bUseGravityOverride=true
    simulationSpace=World
    alphaBoolBlend {blendOption=Linear}
    boundBone {boneName=dyn_shoelace_01_l}
    componentAppliedLinearAccClamp=<1000 1000 1000>
    componentLinearAccScale=<0.5 0.5 0.5>
    componentLinearVelScale=<0.6 0.6 0.6>
    gravityOverride {z=-980}
    physicsBodyDefinitions [{
      boundBone {boneName=dyn_shoelace_01_l}
      boxExtents=<12 12 12>
      constraintSetup {angularLimitsMax=<5 35 7>, angularLimitsMin=<-5 -8 -60>, angularTarget=<0 0 0>}
      localJointOffset=<1 0 0>
     }]
   }
   n2 = Anim Dynamics - Bone: dyn_shoelace_02_l(ComponentPose=@n1, Alpha=1, ExternalForce="0,0,0") {
    angularDampingOverride=0.999
    bOverrideAngularDamping=true
    bUseGravityOverride=true
    simulationSpace=World
    alphaBoolBlend {blendOption=Linear}
    boundBone {boneName=dyn_shoelace_02_l}
    componentAppliedLinearAccClamp=<1000 1000 1000>
    componentLinearAccScale=<0.5 0.5 0.5>
    componentLinearVelScale=<0.6 0.6 0.6>
    gravityOverride {z=-980}
    physicsBodyDefinitions [{
      boundBone {boneName=dyn_shoelace_02_l}
      boxExtents=<10 10 10>
      constraintSetup {angularLimitsMax=<5 35 7>, angularLimitsMin=<-5 -8 -60>, angularTarget=<0 0 0>}
      localJointOffset=<1 0 0>
     }]
   }
   n3 = Anim Dynamics - Bone: dyn_shoelace_01_r(ComponentPose=@n2, Alpha=1, ExternalForce="0,0,0") {
    angularDampingOverride=0.999
    bOverrideAngularDamping=true
    bUseGravityOverride=true
    simulationSpace=World
    alphaBoolBlend {blendOption=Linear}
    boundBone {boneName=dyn_shoelace_01_r}
    componentAppliedLinearAccClamp=<1000 1000 1000>
    componentLinearAccScale=<0.5 0.5 0.5>
    componentLinearVelScale=<0.6 0.6 0.6>
    gravityOverride {z=-980}
    physicsBodyDefinitions [{
      boundBone {boneName=dyn_shoelace_01_r}
      boxExtents=<12 12 12>
      constraintSetup {angularLimitsMax=<5 8 7>, angularLimitsMin=<-5 -20 -60>, angularTarget=<0 0 0>}
      localJointOffset=<1 0 0>
     }]
   }
   n4 = Anim Dynamics - Bone: dyn_shoelace_02_r(ComponentPose=@n3, Alpha=1, ExternalForce="0,0,0") {
    angularDampingOverride=0.999
    bOverrideAngularDamping=true
    bUseGravityOverride=true
    simulationSpace=World
    alphaBoolBlend {blendOption=Linear}
    boundBone {boneName=dyn_shoelace_02_r}
    componentAppliedLinearAccClamp=<1000 1000 1000>
    componentLinearAccScale=<0.5 0.5 0.5>
    componentLinearVelScale=<0.6 0.6 0.6>
    gravityOverride {z=-980}
    physicsBodyDefinitions [{
      boundBone {boneName=dyn_shoelace_02_r}
      boxExtents=<10 10 10>
      constraintSetup {angularLimitsMax=<5 8 7>, angularLimitsMin=<-5 -20 -60>, angularTarget=<0 0 0>}
      localJointOffset=<1 0 0>
     }]
   }
   n5 = Anim Dynamics - Bone: dyn_shoelace_01_l(ComponentPose=Local To Component(LocalPose=Use cached pose 'PhysPose'()), Alpha=1, ExternalForce="0,0,0") {
    angularDampingOverride=0.999
    bOverrideAngularDamping=true
    bUseGravityOverride=true
    simulationSpace=RootRelative
    alphaBoolBlend {blendOption=Linear}
    boundBone {boneName=dyn_shoelace_01_l}
    componentAppliedLinearAccClamp=<1000 1000 5000>
    componentLinearAccScale=<0.5 0.5 0.5>
    componentLinearVelScale=<0.8 0.8 0.6>
    gravityOverride {z=-980}
    physicsBodyDefinitions [{
      boundBone {boneName=dyn_shoelace_01_l}
      boxExtents=<12 12 12>
      constraintSetup {angularLimitsMax=<5 35 7>, angularLimitsMin=<-5 -8 -60>, angularTarget=<0 0 0>}
      localJointOffset=<1 0 0>
     }]
   }
   n6 = Anim Dynamics - Bone: dyn_shoelace_02_l(ComponentPose=@n5, Alpha=1, ExternalForce="0,0,0") {
    angularDampingOverride=0.999
    bOverrideAngularDamping=true
    bUseGravityOverride=true
    simulationSpace=RootRelative
    alphaBoolBlend {blendOption=Linear}
    boundBone {boneName=dyn_shoelace_02_l}
    componentAppliedLinearAccClamp=<1000 1000 5000>
    componentLinearAccScale=<0.5 0.5 0.5>
    componentLinearVelScale=<0.8 0.8 0.6>
    gravityOverride {z=-980}
    physicsBodyDefinitions [{
      boundBone {boneName=dyn_shoelace_02_l}
      boxExtents=<10 10 10>
      constraintSetup {angularLimitsMax=<5 35 7>, angularLimitsMin=<-5 -8 -60>, angularTarget=<0 0 0>}
      localJointOffset=<1 0 0>
     }]
   }
   n7 = Anim Dynamics - Bone: dyn_shoelace_01_r(ComponentPose=@n6, Alpha=1, ExternalForce="0,0,0") {
    angularDampingOverride=0.999
    bOverrideAngularDamping=true
    bUseGravityOverride=true
    simulationSpace=RootRelative
    alphaBoolBlend {blendOption=Linear}
    boundBone {boneName=dyn_shoelace_01_r}
    componentAppliedLinearAccClamp=<1000 1000 5000>
    componentLinearAccScale=<0.5 0.5 0.5>
    componentLinearVelScale=<0.8 0.8 0.6>
    gravityOverride {z=-980}
    physicsBodyDefinitions [{
      boundBone {boneName=dyn_shoelace_01_r}
      boxExtents=<12 12 12>
      constraintSetup {angularLimitsMax=<5 20 7>, angularLimitsMin=<-5 -35 -80>, angularTarget=<0 0 0>}
      localJointOffset=<1 0 0>
     }]
   }
   n8 = Anim Dynamics - Bone: dyn_shoelace_02_r(ComponentPose=@n7, Alpha=1, ExternalForce="0,0,0") {
    angularDampingOverride=0.999
    bOverrideAngularDamping=true
    bUseGravityOverride=true
    simulationSpace=RootRelative
    alphaBoolBlend {blendOption=Linear}
    boundBone {boneName=dyn_shoelace_02_r}
    componentAppliedLinearAccClamp=<1000 1000 5000>
    componentLinearAccScale=<0.5 0.5 0.5>
    componentLinearVelScale=<0.8 0.8 0.6>
    gravityOverride {z=-980}
    physicsBodyDefinitions [{
      boundBone {boneName=dyn_shoelace_02_r}
      boxExtents=<10 10 10>
      constraintSetup {angularLimitsMax=<5 20 7>, angularLimitsMin=<-5 -35 -60>, angularTarget=<0 0 0>}
      localJointOffset=<1 0 0>
     }]
   }
   Output Pose(Result=Blend Poses by bool(BlendPose_0=Component To Local(ComponentPose=@n4), BlendPose_1=Component To Local(ComponentPose=@n8), bActiveValue=Get bIsInFrontEnd()) {blendType=Linear}) {layerGroup=None}
   n9 = "Constraint - dyn_shoelace_MSTR (PP)"!(
    ComponentPose=Local To Component(LocalPose=Copy Pose From Mesh() {bUseAttachedParent=true})
    Alpha=1
    ConstraintWeights_0=0.202
    ConstraintWeights_1:float=0.798
   ) -> (Pose) {
    alphaBoolBlend {blendOption=Linear}
    boneToModify {boneName=dyn_shoelace_MSTR}
    constraintSetup [{targetBone {boneName=calf_r}}, {targetBone {boneName=calf_twist_02_r}}]
    constraintWeights [1,1]
   }
   Save cached pose 'StartPose'(Pose=Component To Local(ComponentPose=@n9)) {cacheName=StartPose}
   n10 = RigidBody(ComponentPose=Local To Component(LocalPose=Use cached pose 'StartPose'())) {
    bClampLinearTranslationLimitToRefPose=true
    bOverrideWorldGravity=true
    overridePhysicsAsset=&/F_MED_RowdyDevilFNCS_Jacket_Physics
    simulationSpace=BaseBoneSpace
    alphaBoolBlend {blendOption=Linear}
    baseBoneRef {boneName=spine_03}
    overrideWorldGravity {z=-980}
    simSpaceSettings {
     dampingAlpha=0
     maxAngularAcceleration=10
     maxAngularVelocity=5
     maxLinearAcceleration=1000
     maxLinearVelocity=500
     worldAlpha=1
     externalLinearDragV {z=0.5}
    }
   }
   n11 = RigidBody(ComponentPose=@n10) {
    bClampLinearTranslationLimitToRefPose=true
    bOverrideWorldGravity=true
    overridePhysicsAsset=&/F_MED_RowdyDevilFNCS_NecklacePhysics
    simulationSpace=BaseBoneSpace
    alphaBoolBlend {blendOption=Linear}
    baseBoneRef {boneName=spine_05}
    overrideWorldGravity {z=-980}
    simSpaceSettings {
     dampingAlpha=0
     maxAngularAcceleration=40
     maxAngularVelocity=10
     maxLinearAcceleration=1100
     maxLinearVelocity=700
     worldAlpha=1
     externalLinearDragV {z=0.5}
    }
   }
   n12 = Control Rig(Source=Component To Local(ComponentPose=@n11)) {
    alphaBoolBlend {blendOption=Linear}
    controlRigAssetReference {blueprintRigClass=&/F_MED_RowdyDevilFNCS_CtrlRig.F_MED_RowdyDevilFNCS_CtrlRig_C}
   }
   Save cached pose 'PhysPose'(Pose=@n12) {cacheName=PhysPose}
  ]}
 EventGraph {nodes [Event BlueprintUpdateAnimation()]}
]
variables []
