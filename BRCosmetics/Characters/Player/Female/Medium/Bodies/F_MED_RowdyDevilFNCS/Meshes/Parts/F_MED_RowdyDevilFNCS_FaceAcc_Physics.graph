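# Physics asset for the face accessory of the RowdyDevilFNCS female medium cosmetic character, covering its left and right dynamic tail bones.
# Each side is a three-capsule chain hanging off a small head sphere, linked parent to child by angular limits.
# Twist is nearly locked and swing is asymmetric, tightest in one axis at the head joint.
# Unconstrained clavicle capsules supply shoulder collision volumes, with an extra capsule on the right.
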PhysicsAsset /BRCosmetics/Characters/Player/Female/Medium/Bodies/F_MED_RowdyDevilFNCS/Meshes/Parts/F_MED_RowdyDevilFNCS_FaceAcc_Physics
bodies {
 clavicle_l [{
   shapeType=Capsule
   length=7.5
   radius=8.5
   shapeName=clavicle_l_capsule
   center=<6.1731 1.3279 -2.3386>
   rotation=<rot -73.5558 118.3593 -121.3902>
  }]
 clavicle_r [
  {
   shapeType=Capsule
   length=7.5
   radius=8.5
   shapeName=clavicle_r_capsule
   center=<-6.1731 -1.3279 2.3386>
   rotation=<rot -73.5558 118.3593 -121.3902>
  }
  {
   shapeType=Capsule
   length=4
   radius=4.5
   shapeName=clavicle_r_capsule
   center=<-13.5084 -3.3088 -7.5322>
   rotation=<rot 0 -179.9989 -179.9997>
  }
 ]
 dyn_tail_l_01 [{shapeType=Capsule, length=5, radius=3, shapeName=dyn_tail_l_01_capsule, center=<1.9094 0 0>, rotation=<rot -90 0 0>}]
 dyn_tail_l_02 [{shapeType=Capsule, length=5, radius=3, shapeName=dyn_tail_l_02_capsule, center=<1.9094 0 0>, rotation=<rot -90 0 0>}]
 dyn_tail_l_03 [{shapeType=Capsule, length=5, radius=3, shapeName=dyn_tail_l_03_capsule, center=<1.9094 0 0>, rotation=<rot -90 -0.0003 0>}]
 dyn_tail_r_01 [{shapeType=Capsule, length=2.5, radius=3, shapeName=dyn_tail_r_01_capsule, center=<1.9094 0 0>, rotation=<rot -90 0.0002 0>}]
 dyn_tail_r_02 [{shapeType=Capsule, length=5, radius=3, shapeName=dyn_tail_r_02_capsule, center=<1.9094 0 0>, rotation=<rot -90 0.0002 0>}]
 dyn_tail_r_03 [{shapeType=Capsule, length=4.5, radius=3.5, shapeName=dyn_tail_r_03_capsule, center=<1.9094 0 0>, rotation=<rot -90 0.0003 0>}]
 head [{shapeType=Sphere, radius=1, shapeName=head_sphere}]
}
constraints [
 {bone1Name=dyn_tail_l_01, bone2Name=head, swing1LimitDegrees=5, swing2LimitDegrees=35, twistLimitDegrees=1}
 {bone1Name=dyn_tail_r_01, bone2Name=head, swing1LimitDegrees=5, swing2LimitDegrees=35, twistLimitDegrees=1}
 {bone1Name=dyn_tail_l_02, bone2Name=dyn_tail_l_01, swing1LimitDegrees=15, swing2LimitDegrees=35, twistLimitDegrees=1}
 {bone1Name=dyn_tail_l_03, bone2Name=dyn_tail_l_02, swing1LimitDegrees=15, swing2LimitDegrees=25, twistLimitDegrees=1}
 {bone1Name=dyn_tail_r_02, bone2Name=dyn_tail_r_01, swing1LimitDegrees=15, swing2LimitDegrees=35, twistLimitDegrees=1}
 {bone1Name=dyn_tail_r_03, bone2Name=dyn_tail_r_02, swing1LimitDegrees=15, swing2LimitDegrees=25, twistLimitDegrees=1}
]
