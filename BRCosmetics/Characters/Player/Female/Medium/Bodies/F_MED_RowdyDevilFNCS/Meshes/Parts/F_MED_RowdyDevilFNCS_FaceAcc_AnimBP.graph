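# Post-process-style AnimBP for the RowdyDevilFNCS face accessory, adding secondary motion to hat and bangs.
# Copies the attached parent pose, layering in BodyMesh (found on customization load) from clavicle_r.
# A RigidBody pass in spine_04 space feeds chained per-bone AnimDynamics with damped angular limits, some sprung.
# EventGraph sets Gravity from skydive and parachute state, but the shown AnimGraph uses fixed gravity values.
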AnimBlueprint /BRCosmetics/Characters/Player/Female/Medium/Bodies/F_MED_RowdyDevilFNCS/Meshes/Parts/F_MED_RowdyDevilFNCS_FaceAcc_AnimBP
target_skeleton=&/F_MED_RowdyDevilFNCS_FaceAcc_Skeleton
graphs [
 AnimGraph {nodes [
   n1 = Copy Pose From Mesh() {bUseAttachedParent=true}
   n2 = Copy Pose From Mesh(SourceMeshComponent=Get BodyMesh() -> (BodyMesh:Skeletal Mesh Component Object Reference)) {bUseAttachedParent=true}
   n3 = RigidBody!(
    ComponentPose=Local To Component(LocalPose=Layered blend per bone(BasePose=@n1, BlendPoses_0=@n2) {layerSetup [{branchFilters [{boneName=clavicle_r}, {boneName=clavicle_r}]}]})
    Alpha
    OverrideWorldGravity:Vector="0,0,-600"
    ExternalForce
   ) -> (Pose) {
    bClampLinearTranslationLimitToRefPose=true
    bOverrideWorldGravity=true
    overridePhysicsAsset=&/F_MED_RowdyDevilFNCS_FaceAcc_Physics
    simulationSpace=BaseBoneSpace
    alphaBoolBlend {blendOption=Linear}
    baseBoneRef {boneName=spine_04}
    componentAppliedLinearAccClamp=<2500 2500 2500>
    componentLinearAccScale=<0.2 0.2 0.2>
    componentLinearVelScale=<0.2 0.2 0.2>
    overrideWorldGravity {z=-600}
   }
   n4 = Anim Dynamics - Bone: dyn_hat(ComponentPose=@n3, Alpha=1, ExternalForce="0,0,0") {
    angularDampingOverride=0.999
    bOverrideAngularDamping=true
    bUseGravityOverride=true
    alphaBoolBlend {blendOption=Linear}
    boundBone {boneName=dyn_hat}
    componentAppliedLinearAccClamp=<2000 2000 2000>
    componentLinearAccScale=<0.2 0.2 0.2>
    componentLinearVelScale=<0.2 0.2 0.2>
    gravityOverride {z=-900}
    physicsBodyDefinitions [{
      sphereCollisionRadius=0
      boundBone {boneName=dyn_hat}
      boxExtents=<12 10 10>
      constraintSetup {angularLimitsMax=<0 20 25>, angularLimitsMin=<0 -20 0>, angularTarget=<1 0 0>}
      localJointOffset=<3 0 0>
     }]
   }
   n5 = Anim Dynamics - Bone: dyn_bang_b(ComponentPose=@n4, Alpha=1, ExternalForce="0,0,0") {
    angularDampingOverride=0.999
    bOverrideAngularDamping=true
    bUseGravityOverride=true
    alphaBoolBlend {blendOption=Linear}
    boundBone {boneName=dyn_bang_b}
    componentAppliedLinearAccClamp=<5000 5000 5000>
    componentLinearAccScale=<0.2 0.2 0.2>
    componentLinearVelScale=<0.2 0.2 0.2>
    gravityOverride {z=-900}
    physicsBodyDefinitions [{
      sphereCollisionRadius=0
      boundBone {boneName=dyn_bang_b}
      boxExtents=<12 10 10>
      constraintSetup {angularLimitsMax=<0 0 10>, angularLimitsMin=<0 -10 0>, angularTarget=<1 0 0>}
      localJointOffset=<3 0 0>
     }]
   }
   n6 = Anim Dynamics - Bone: dyn_bang_b_01(ComponentPose=@n5, Alpha=1, ExternalForce="0,0,0") {
    angularDampingOverride=0.999
    bOverrideAngularDamping=true
    bUseGravityOverride=true
    alphaBoolBlend {blendOption=Linear}
    boundBone {boneName=dyn_bang_b_01}
    componentAppliedLinearAccClamp=<5000 5000 5000>
    componentLinearAccScale=<0.2 0.2 0.2>
    componentLinearVelScale=<0.2 0.2 0.2>
    gravityOverride {z=-900}
    physicsBodyDefinitions [{
      sphereCollisionRadius=0
      boundBone {boneName=dyn_bang_b_01}
      boxExtents=<12 10 10>
      constraintSetup {angularLimitsMax=<0 0 5>, angularLimitsMin=<0 -5 -5>, angularTarget=<1 0 0>}
      localJointOffset=<2 0 0>
     }]
   }
   n7 = Anim Dynamics - Bone: dyn_bang_a(ComponentPose=@n6, Alpha=1, ExternalForce="0,0,0") {
    angularDampingOverride=0.99
    angularSpringConstant=1000
    bOverrideAngularDamping=true
    bUseGravityOverride=true
    alphaBoolBlend {blendOption=Linear}
    boundBone {boneName=dyn_bang_a}
    componentAppliedLinearAccClamp=<5000 5000 5000>
    componentLinearAccScale=<0.3 0.3 0.3>
    componentLinearVelScale=<0.3 0.3 0.3>
    gravityOverride {z=-980}
    physicsBodyDefinitions [{
      boundBone {boneName=dyn_bang_a}
      boxExtents=<10 15 15>
      constraintSetup {angularLimitsMax=<0 10 15>, angularLimitsMin=<0 -10 0>, angularTarget=<1 0 0>}
      localJointOffset=<2 0 0>
     }]
   }
   n8 = Anim Dynamics - Bone: dyn_bang_c(ComponentPose=@n7, Alpha=1, ExternalForce="0,0,0") {
    angularDampingOverride=0.99
    angularSpringConstant=1000
    bAngularSpring=true
    bOverrideAngularDamping=true
    bUseGravityOverride=true
    bUsePlanarLimit=false
    gravityScale=0.35
    alphaBoolBlend {blendOption=Linear}
    boundBone {boneName=dyn_bang_c}
    componentAppliedLinearAccClamp=<400 400 400>
    componentLinearAccScale=<0.3 0.3 0.3>
    componentLinearVelScale=<0.4 0.4 0.4>
    gravityOverride {z=-600}
    physicsBodyDefinitions [{
      sphereCollisionRadius=0
      boundBone {boneName=dyn_bang_c}
      boxExtents=<10 10 10>
      constraintSetup {angularLimitsMax=<0 0 10>, angularLimitsMin=<0 -25 -10>, angularTarget=<1 0 0>}
      localJointOffset=<1 0 0>
     }]
   }
   n9 = Anim Dynamics - Bone: dyn_bang_d(ComponentPose=@n8, Alpha=1, ExternalForce="0,0,0") {
    angularDampingOverride=0.99
    angularSpringConstant=1000
    bAngularSpring=true
    bOverrideAngularDamping=true
    bUseGravityOverride=true
    bUsePlanarLimit=false
    gravityScale=0.35
    alphaBoolBlend {blendOption=Linear}
    boundBone {boneName=dyn_bang_d}
    componentAppliedLinearAccClamp=<600 600 600>
    componentLinearAccScale=<0.3 0.3 0.3>
    componentLinearVelScale=<0.4 0.4 0.4>
    gravityOverride {z=-400}
    physicsBodyDefinitions [{
      sphereCollisionRadius=0
      boundBone {boneName=dyn_bang_d}
      boxExtents=<15 5 5>
      constraintSetup {angularLimitsMax=<0 0 15>, angularLimitsMin=<0 -25 -15>, angularTarget=<1 0 0>}
      localJointOffset=<1 0 0>
     }]
   }
   n10 = Anim Dynamics - Bone: dyn_bang_e(ComponentPose=@n9, Alpha=1, ExternalForce="0,0,0") {
    angularDampingOverride=0.99
    angularSpringConstant=1000
    bAngularSpring=true
    bOverrideAngularDamping=true
    bUseGravityOverride=true
    bUsePlanarLimit=false
    gravityScale=0.35
    alphaBoolBlend {blendOption=Linear}
    boundBone {boneName=dyn_bang_e}
    componentAppliedLinearAccClamp=<1500 1500 1500>
    componentLinearAccScale=<0.3 0.3 0.3>
    componentLinearVelScale=<0.4 0.4 0.4>
    gravityOverride {z=-400}
    physicsBodyDefinitions [{
      sphereCollisionRadius=1
      boundBone {boneName=dyn_bang_e}
      boxExtents=<15 10 10>
      constraintSetup {angularLimitsMax=<0 25 25>, angularLimitsMin=<0 -20 0>, angularTarget=<1 0 0>}
      localJointOffset=<1.5 0 0>
     }]
   }
   n11 = Anim Dynamics - Bone: dyn_bang_e_01(ComponentPose=@n10, Alpha=1, ExternalForce="0,0,0") {
    angularDampingOverride=0.99
    angularSpringConstant=1000
    bAngularSpring=true
    bOverrideAngularDamping=true
    bUseGravityOverride=true
    bUsePlanarLimit=false
    gravityScale=0.35
    alphaBoolBlend {blendOption=Linear}
    boundBone {boneName=dyn_bang_e_01}
    componentAppliedLinearAccClamp=<1500 1500 1500>
    componentLinearAccScale=<0.3 0.3 0.3>
    componentLinearVelScale=<0.4 0.4 0.4>
    gravityOverride {z=-400}
    physicsBodyDefinitions [{
      sphereCollisionRadius=1
      boundBone {boneName=dyn_bang_e_01}
      boxExtents=<15 10 10>
      constraintSetup {angularLimitsMax=<0 25 25>, angularLimitsMin=<0 -20 0>, angularTarget=<1 0 0>}
      localJointOffset=<1.5 0 0>
     }]
   }
   n12 = Anim Dynamics - Bone: dyn_bang_f(ComponentPose=@n11, Alpha=1, ExternalForce="0,0,0") {
    angularDampingOverride=0.99
    angularSpringConstant=1000
    bAngularSpring=true
    bOverrideAngularDamping=true
    bUseGravityOverride=true
    bUsePlanarLimit=false
    gravityScale=0.35
    alphaBoolBlend {blendOption=Linear}
    boundBone {boneName=dyn_bang_f}
    componentAppliedLinearAccClamp=<550 550 550>
    componentLinearAccScale=<0.3 0.3 0.3>
    componentLinearVelScale=<0.4 0.4 0.4>
    gravityOverride {z=-400}
    physicsBodyDefinitions [{
      sphereCollisionRadius=1
      boundBone {boneName=dyn_bang_f}
      boxExtents=<15 10 10>
      constraintSetup {angularLimitsMax=<0 25 20>, angularLimitsMin=<0 -20 0>, angularTarget=<1 0 0>}
      localJointOffset=<1.5 0 0>
     }]
   }
   Output Pose(Result=Component To Local(ComponentPose=@n12)) {layerGroup=None}
  ]}
 EventGraph {nodes [
   Event BlueprintUpdateAnimation()
   n1 = Get bIsParachuteOpened()
   > Set Gravity(Gravity:Vector=SelectVector(A=SelectVector(A="0,0,-200", B="0,0,150", bPickA=@n1), B="0,0,-600", bPickA=Get bIsSkyDiving())) -> (Output_Get:Vector)
   n2 = Event OnCharacterCustomizationLoaded()
   > Set BodyMesh(BodyMesh=Get Array Item with EFort Custom Part Type(Array=@n2.CustomMeshParts))
  ]}
]
variables [
 BodyMesh {type=SkeletalMeshComponent}
 Gravity {type=Vector}
]
